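# Physics asset for the OrinBlitz medium male player body cosmetic, apparently built to simulate its dynamic scarf.
# Capsules on the pelvis, spine and right arm form a partial body for the scarf, joined mostly by free-swing constraints.
# A five-bone scarf chain hangs off the right clavicle and upper spine, with zero-degree twist limits and primary swing limits that tighten toward the tip.
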PhysicsAsset /BRCosmetics/Characters/Player/Male/Medium/Bodies/M_MED_OrinBlitz/Meshes/M_MED_OrinBlitz_Physics
bodies {
 clavicle_r [{shapeType=Capsule, length=8, radius=10, shapeName=clavicle_r_capsule, center=<-5 0 -3>, rotation=<rot 90 0 0>}]
 dyn_scarf_01 [{shapeType=Capsule, length=6.7557, radius=5, shapeName=dyn_scarf_01_capsule, center=<5 -4 0>, rotation=<rot 90 0 0>}]
 dyn_scarf_02 [{shapeType=Capsule, length=5.5, radius=5, shapeName=dyn_scarf_02_capsule, center=<5 -1 0>, rotation=<rot 90 0 0>}]
 dyn_scarf_03 [{shapeType=Capsule, length=4, radius=5.6, shapeName=dyn_scarf_03_capsule, center=<5 0 0>, rotation=<rot 90 0 0>}]
 dyn_scarf_04 [{shapeType=Capsule, length=4, radius=5.6, shapeName=dyn_scarf_04_capsule, center=<6 0 0>, rotation=<rot 90 0 0>}]
 dyn_scarf_base [{shapeType=Capsule, length=2, radius=2, shapeName=dyn_scarf_base_capsule}]
 lowerarm_r [{shapeType=Capsule, length=18, radius=5, shapeName=lowerarm_r_capsule, center=<-8.997 0.0911 0>, rotation=<rot 90 -1.8909 0>}]
 pelvis [{shapeType=Capsule, length=9.0046, radius=13.9954, shapeName=pelvis_capsule, center=<-3 0 0>}]
 spine_01 [{shapeType=Capsule, length=3.9561, radius=11.9852, shapeName=spine_01_capsule}]
 spine_02 [{shapeType=Capsule, length=3.9561, radius=11.9852, shapeName=spine_02_capsule}]
 spine_03 [{shapeType=Capsule, length=3.9561, radius=11.9852, shapeName=spine_03_capsule}]
 spine_04 [{shapeType=Capsule, length=5.2478, radius=13.5024, shapeName=spine_04_capsule, center=<4.1761 1.4558 0>}]
 spine_05 [{shapeType=Capsule, length=15.9961, radius=15, shapeName=spine_05_capsule, center=<0.3168 3.6072 0>}]
 upperarm_r [{shapeType=Capsule, length=18.382, radius=7.3544, shapeName=upperarm_r_capsule, center=<-5 0 0>, rotation=<rot 90 0 0>}]
}
constraints [
 {bone1Name=spine_03, bone2Name=pelvis, swing1LimitDegrees=45, swing2LimitDegrees=45, twistLimitDegrees=45}
 {bone1Name=spine_05, bone2Name=spine_03, swing1LimitDegrees=45, swing2LimitDegrees=45, twistLimitDegrees=45}
 {bone1Name=dyn_scarf_01, bone2Name=spine_05, swing1LimitDegrees=45, swing2LimitDegrees=45, twistLimitDegrees=0}
 {
  bone1Name=spine_01
  bone2Name=pelvis
  swing1LimitDegrees=45
  swing1Motion=Free
  swing2LimitDegrees=45
  swing2Motion=Free
  twistLimitDegrees=45
  twistMotion=Free
 }
 {
  bone1Name=spine_02
  bone2Name=spine_01
  swing1LimitDegrees=45
  swing1Motion=Free
  swing2LimitDegrees=45
  swing2Motion=Free
  twistLimitDegrees=45
  twistMotion=Free
 }
 {
  bone1Name=spine_03
  bone2Name=spine_02
  swing1LimitDegrees=45
  swing1Motion=Free
  swing2LimitDegrees=45
  swing2Motion=Free
  twistLimitDegrees=45
  twistMotion=Free
 }
 {
  bone1Name=spine_04
  bone2Name=spine_03
  swing1LimitDegrees=45
  swing1Motion=Free
  swing2LimitDegrees=45
  swing2Motion=Free
  twistLimitDegrees=45
  twistMotion=Free
 }
 {
  bone1Name=spine_05
  bone2Name=spine_04
  swing1LimitDegrees=45
  swing1Motion=Free
  swing2LimitDegrees=45
  swing2Motion=Free
  twistLimitDegrees=45
  twistMotion=Free
 }
 {
  bone1Name=clavicle_r
  bone2Name=spine_05
  swing1LimitDegrees=45
  swing1Motion=Free
  swing2LimitDegrees=45
  swing2Motion=Free
  twistLimitDegrees=45
  twistMotion=Free
 }
 {
  bone1Name=dyn_scarf_base
  bone2Name=clavicle_r
  swing1LimitDegrees=45
  swing1Motion=Free
  swing2LimitDegrees=45
  swing2Motion=Free
  twistLimitDegrees=45
  twistMotion=Free
 }
 {bone1Name=dyn_scarf_01, bone2Name=dyn_scarf_base, swing1LimitDegrees=20, swing2LimitDegrees=2, twistLimitDegrees=0}
 {bone1Name=dyn_scarf_02, bone2Name=dyn_scarf_01, swing1LimitDegrees=7.5, swing2LimitDegrees=2, twistLimitDegrees=0}
 {bone1Name=dyn_scarf_03, bone2Name=dyn_scarf_02, swing1LimitDegrees=5, swing2LimitDegrees=2, twistLimitDegrees=0}
 {bone1Name=dyn_scarf_04, bone2Name=dyn_scarf_03, swing1LimitDegrees=2.5, swing2LimitDegrees=2, twistLimitDegrees=0}
 {
  bone1Name=upperarm_r
  bone2Name=clavicle_r
  swing1LimitDegrees=45
  swing1Motion=Free
  swing2LimitDegrees=45
  swing2Motion=Free
  twistLimitDegrees=45
  twistMotion=Free
 }
 {
  bone1Name=lowerarm_r
  bone2Name=upperarm_r
  swing1LimitDegrees=45
  swing1Motion=Free
  swing2LimitDegrees=45
  swing2Motion=Free
  twistLimitDegrees=45
  twistMotion=Free
 }
 {
  bone1Name=dyn_scarf_base
  bone2Name=spine_04
  swing1LimitDegrees=45
  swing1Motion=Free
  swing2LimitDegrees=45
  swing2Motion=Free
  twistLimitDegrees=45
  twistMotion=Free
 }
 {
  bone1Name=clavicle_r
  bone2Name=spine_04
  swing1LimitDegrees=45
  swing1Motion=Free
  swing2LimitDegrees=45
  swing2Motion=Free
  twistLimitDegrees=45
  twistMotion=Free
 }
]
