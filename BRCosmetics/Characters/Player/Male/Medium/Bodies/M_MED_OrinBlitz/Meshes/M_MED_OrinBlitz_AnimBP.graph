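# Post-process style AnimBP for the M_MED_OrinBlitz player body cosmetic, adding scarf motion to the parent mesh's copied pose.
# A Control Rig feeds two cached-pose branches, switched by crouching: a plain RigidBody sim, or layered scarf bone offsets, RigidBody and AnimDynamics.
# The event graph derives weapon, front-end movement and backbling state; any backbling zeroes the layered branch's RigidBody alpha.
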AnimBlueprint /BRCosmetics/Characters/Player/Male/Medium/Bodies/M_MED_OrinBlitz/Meshes/M_MED_OrinBlitz_AnimBP
target_skeleton=&/M_MED_OrinBlitz_Skeleton
graphs [
 AnimGraph {nodes [
   n1 = RigidBody!(
    ComponentPose=Local To Component(LocalPose=Use cached pose 'Ctrl Output'())
    Alpha
    OverrideWorldGravity:Vector="0,0,0"
    ExternalForce
    SimSpaceSettings:Sim Space Settings Structure=(WorldAlpha=0.5,VelocityScaleZ=0,DampingAlpha=0.5,MaxLinearVelocity=200,MaxAngularVelocity=5,MaxLinearAcceleration=1000,MaxAngularAcceleration=200,ExternalLinearDragV=(X=1,Y=1,Z=1),ExternalLinearVelocity=(X=0,Y=0,Z=0),ExternalAngularVelocity=(X=0,Y=0,Z=0))
   ) -> (Pose) {
    bClampLinearTranslationLimitToRefPose=true
    bOverrideWorldGravity=true
    overridePhysicsAsset=&/M_MED_OrinBlitz_Physics
    simulationSpace=BaseBoneSpace
    alphaBoolBlend {blendOption=Linear}
    baseBoneRef {boneName=spine_01}
    componentAppliedLinearAccClamp=<3000 3000 3000>
    componentLinearAccScale=<0.3 0.2 0.2>
    componentLinearVelScale=<0.3 0.2 0.2>
    overrideWorldGravity {z=-980}
    simSpaceSettings {
     dampingAlpha=0.5
     maxAngularAcceleration=200
     maxAngularVelocity=5
     maxLinearAcceleration=1000
     maxLinearVelocity=200
     velocityScaleZ=0
     worldAlpha=0.5
     externalLinearDragV=<1 1 1>
    }
   }
   Output Pose(Result=Blend Poses by bool(
     BlendPose_0=Component To Local(ComponentPose=@n1)
     BlendPose_1=Use cached pose 'Body Dynamics'()
     bActiveValue=Get Crouch Walk() -> (Crouch Walk:bool)
    )) {layerGroup=None}
   n2 = Control Rig!(
    Source=Copy Pose From Mesh() {bUseAttachedParent=true}
    Alpha
    Playing Emote:bool=Get bIsPlayingEmote()
    RifleEquipped:double=Get isUsingRifile() -> (isUsingRifile:double)
   ) -> (Pose) {controlRigAssetReference {blueprintRigClass=&/M_MED_OrinBlitz_CtrlRig_.M_MED_OrinBlitz_CtrlRig__C}}
   Save cached pose 'Ctrl Output'(Pose=@n2) {cacheName=Ctrl Output}
   n3 = "Transform (Modify) Bone - Bone: dyn_scarf_base"(
    ComponentPose=Local To Component(LocalPose=Use cached pose 'Ctrl Output'())
    Alpha="To Float (Boolean)"(InBool=Get 2d moding() -> (2d moding:bool))
    Translation="-2,-4,0"
    Rotation="0,-5,5"
    Scale="1,1,1"
   ) {
    rotationMode=BMM_Additive
    rotationSpace=BCS_BoneSpace
    translationMode=BMM_Additive
    translationSpace=BCS_BoneSpace
    alphaScaleBiasClamp {bInterpResult=true, interpSpeedDecreasing=2, interpSpeedIncreasing=2}
    boneToModify {boneName=dyn_scarf_base}
    rotation {roll=5, yaw=-5}
    translation=<-2 -4>
   }
   n4 = "Transform (Modify) Bone - Bone: dyn_scarf_base"(ComponentPose=@n3, Alpha="To Float (Boolean)"(InBool=Get bIsPlayingMeleeAnim()), Translation="0,-3,0", Rotation="0,0,0", Scale="1,1,1") {
    rotationMode=BMM_Additive
    rotationSpace=BCS_BoneSpace
    translationMode=BMM_Additive
    translationSpace=BCS_BoneSpace
    alphaScaleBiasClamp {bInterpResult=true, interpSpeedDecreasing=2, interpSpeedIncreasing=2}
    boneToModify {boneName=dyn_scarf_base}
    translation {y=-3}
   }
   n5 = RigidBody(
    ComponentPose=@n4
    Alpha=Get Wear A Cape Alpha() -> (Wear A Cape Alpha:double)
    OverrideWorldGravity="0,0,0"
    SimSpaceSettings=(WorldAlpha=0.5,VelocityScaleZ=0,DampingAlpha=0.5,MaxLinearVelocity=200,MaxAngularVelocity=5,MaxLinearAcceleration=1000,MaxAngularAcceleration=200,ExternalLinearDragV=(X=1,Y=1,Z=1),ExternalLinearVelocity=(X=0,Y=0,Z=0),ExternalAngularVelocity=(X=0,Y=0,Z=0))
   ) {
    bClampLinearTranslationLimitToRefPose=true
    bOverrideWorldGravity=true
    overridePhysicsAsset=&/M_MED_OrinBlitz_Physics_Check
    simulationSpace=BaseBoneSpace
    alphaBoolBlend {blendOption=Linear}
    baseBoneRef {boneName=spine_01}
    componentAppliedLinearAccClamp=<3000 3000 3000>
    componentLinearAccScale=<0.3 0.2 0.2>
    componentLinearVelScale=<0.3 0.2 0.2>
    overrideWorldGravity {z=-980}
    simSpaceSettings {
     dampingAlpha=0.5
     maxAngularAcceleration=200
     maxAngularVelocity=5
     maxLinearAcceleration=1000
     maxLinearVelocity=200
     velocityScaleZ=0
     worldAlpha=0.5
     externalLinearDragV=<1 1 1>
    }
   }
   n6 = "Transform (Modify) Bone - Bone: dyn_scarf_01"!(ComponentPose=@n5, bAlphaBoolEnabled:bool=True, Translation="-0.144,-0.060,-3.266", Rotation="12.478,8.746,-0.472", Scale="1,1,1") -> (Pose) {
    alphaInputType=Bool
    rotationMode=BMM_Additive
    rotationSpace=BCS_BoneSpace
    translationMode=BMM_Additive
    translationSpace=BCS_BoneSpace
    alphaBoolBlend {blendInTime=0.2, blendOutTime=0.2}
    boneToModify {boneName=dyn_scarf_01}
    rotation=<rot 12.478 8.746 -0.472>
    translation=<-0.144 -0.06 -3.266>
   }
   n7 = Anim Dynamics - Bone: dyn_scarf_front_01(ComponentPose=@n6, Alpha=1, ExternalForce="0,0,0") {
    angularDampingOverride=1
    angularSpringConstant=20000
    bAngularSpring=true
    bOverrideAngularDamping=true
    bUseGravityOverride=true
    gravityScale=0.35
    simulationSpace=RootRelative
    boundBone {boneName=dyn_scarf_front_01}
    componentLinearAccScale=<1 1 1>
    componentLinearVelScale=<1 1 1>
    gravityOverride {z=-980}
    physicsBodyDefinitions [{
      boundBone {boneName=dyn_scarf_front_01}
      boxExtents=<20 30 30>
      constraintSetup {angularLimitsMax=<0 0 10>, angularLimitsMin=<0 -50 -10>, angularTarget=<1 0 0>}
      localJointOffset=<5 0 0>
     }]
   }
   n8 = "Transform (Modify) Bone - Bone: dyn_scarf_01"(ComponentPose=@n7, bAlphaBoolEnabled=True, Translation="0,0,0", Rotation="0,10,0", Scale="1,1,1") {
    alphaInputType=Bool
    rotationMode=BMM_Additive
    rotationSpace=BCS_BoneSpace
    translationMode=BMM_Additive
    translationSpace=BCS_BoneSpace
    alphaBoolBlend {blendInTime=1, blendOutTime=1}
    boneToModify {boneName=dyn_scarf_01}
    rotation {yaw=10}
   }
   n9 = "Transform (Modify) Bone - Bone: dyn_scarf_02"!(ComponentPose=@n8, bAlphaBoolEnabled:bool=True, Translation="0,0,0", Rotation="0,5,0", Scale="1,1,1") -> (Pose) {
    alphaInputType=Bool
    rotationMode=BMM_Additive
    rotationSpace=BCS_BoneSpace
    translationMode=BMM_Additive
    translationSpace=BCS_BoneSpace
    alphaBoolBlend {blendInTime=1, blendOutTime=1}
    boneToModify {boneName=dyn_scarf_02}
    rotation {yaw=5}
   }
   n10 = "Transform (Modify) Bone - Bone: dyn_scarf_03"!(ComponentPose=@n9, bAlphaBoolEnabled:bool=True, Translation="0,0,0", Rotation="0,9,0", Scale="1,1,1") -> (Pose) {
    alphaInputType=Bool
    rotationMode=BMM_Additive
    rotationSpace=BCS_BoneSpace
    translationMode=BMM_Additive
    translationSpace=BCS_BoneSpace
    alphaBoolBlend {blendInTime=1, blendOutTime=1}
    boneToModify {boneName=dyn_scarf_03}
    rotation {yaw=9}
   }
   n11 = "Transform (Modify) Bone - Bone: dyn_scarf_04"!(ComponentPose=@n10, bAlphaBoolEnabled:bool=True, Translation="0,0,0", Rotation="0,-5,0", Scale="1,1,1") -> (Pose) {
    alphaInputType=Bool
    rotationMode=BMM_Additive
    rotationSpace=BCS_BoneSpace
    translationMode=BMM_Additive
    translationSpace=BCS_BoneSpace
    alphaBoolBlend {blendInTime=1, blendOutTime=1}
    boneToModify {boneName=dyn_scarf_04}
    rotation {yaw=-5}
   }
   n12 = "Transform (Modify) Bone - Bone: dyn_scarf_01"(ComponentPose=@n11, bAlphaBoolEnabled=True, Translation="0,0,0", Rotation="0,6,-20", Scale="1,1,1") {
    alphaInputType=Bool
    rotationMode=BMM_Additive
    rotationSpace=BCS_BoneSpace
    translationMode=BMM_Additive
    translationSpace=BCS_BoneSpace
    alphaBoolBlend {blendInTime=1, blendOutTime=1}
    boneToModify {boneName=dyn_scarf_01}
    rotation {roll=-20, yaw=6}
   }
   n13 = "Transform (Modify) Bone - Bone: dyn_scarf_02"(ComponentPose=@n12, bAlphaBoolEnabled=True, Translation="0,0,0", Rotation="0,0,0", Scale="1,1,1") {
    alphaInputType=Bool
    rotationMode=BMM_Additive
    rotationSpace=BCS_BoneSpace
    translationMode=BMM_Additive
    translationSpace=BCS_BoneSpace
    alphaBoolBlend {blendInTime=1, blendOutTime=1}
    boneToModify {boneName=dyn_scarf_02}
   }
   n14 = "Transform (Modify) Bone - Bone: dyn_scarf_03"(ComponentPose=@n13, bAlphaBoolEnabled=True, Translation="0,0,0", Rotation="0,8,0", Scale="1,1,1") {
    alphaInputType=Bool
    rotationMode=BMM_Additive
    rotationSpace=BCS_BoneSpace
    translationMode=BMM_Additive
    translationSpace=BCS_BoneSpace
    alphaBoolBlend {blendInTime=1, blendOutTime=1}
    boneToModify {boneName=dyn_scarf_03}
    rotation {yaw=8}
   }
   n15 = "Transform (Modify) Bone - Bone: dyn_scarf_04"(ComponentPose=@n14, bAlphaBoolEnabled=True, Translation="0,0,0", Rotation="0,-5,0", Scale="1,1,1") {
    alphaInputType=Bool
    rotationMode=BMM_Additive
    rotationSpace=BCS_BoneSpace
    translationMode=BMM_Additive
    translationSpace=BCS_BoneSpace
    alphaBoolBlend {blendInTime=1, blendOutTime=1}
    boneToModify {boneName=dyn_scarf_04}
    rotation {yaw=-5}
   }
   n16 = Anim Dynamics - Bone: dyn_scarf_01!(ComponentPose=@n15, Alpha=0.5, GravityOverride:Vector="0,0,0", ExternalForce="0,0,0") -> (Pose) {
    angularDampingOverride=1
    angularSpringConstant=3000
    bAngularSpring=true
    bOverrideAngularDamping=true
    bUseGravityOverride=true
    simulationSpace=RootRelative
    alphaBoolBlend {blendOption=Linear}
    boundBone {boneName=dyn_scarf_01}
    componentAppliedLinearAccClamp=<1000 1000 1000>
    componentLinearAccScale=<0.8 0.8 0.8>
    componentLinearVelScale=<0.8 0.8 0.8>
    physicsBodyDefinitions [{
      boundBone {boneName=dyn_scarf_01}
      boxExtents=<40 40 40>
      constraintSetup {angularLimitsMax=<0 5 0>, angularLimitsMin=<0 -5 -5>, angularTarget=<0 0 0>}
      localJointOffset=<6 0 0>
     }]
    relativeSpaceBone {boneName=pelvis}
   }
   n17 = Anim Dynamics - Bone: dyn_scarf_02!(ComponentPose=@n16, Alpha=0.5, GravityOverride:Vector="0,0,0", ExternalForce="0,0,0") -> (Pose) {
    angularDampingOverride=1
    angularSpringConstant=3000
    bAngularSpring=true
    bOverrideAngularDamping=true
    bUseGravityOverride=true
    simulationSpace=RootRelative
    alphaBoolBlend {blendOption=Linear}
    boundBone {boneName=dyn_scarf_02}
    componentAppliedLinearAccClamp=<1000 1000 1000>
    componentLinearAccScale=<0.8 0.8 0.8>
    componentLinearVelScale=<0.8 0.8 0.8>
    physicsBodyDefinitions [{
      boundBone {boneName=dyn_scarf_02}
      boxExtents=<40 40 40>
      constraintSetup {angularLimitsMax=<0 5 0>, angularLimitsMin=<0 -5 -5>, angularTarget=<0 0 0>}
      localJointOffset=<6 0 0>
     }]
    relativeSpaceBone {boneName=pelvis}
   }
   n18 = Anim Dynamics - Bone: dyn_scarf_02(ComponentPose=@n17, Alpha=0.5, GravityOverride="0,0,0", ExternalForce="0,0,0") {
    angularDampingOverride=1
    angularSpringConstant=3000
    bAngularSpring=true
    bOverrideAngularDamping=true
    bUseGravityOverride=true
    simulationSpace=RootRelative
    alphaBoolBlend {blendOption=Linear}
    boundBone {boneName=dyn_scarf_02}
    componentAppliedLinearAccClamp=<1000 1000 1000>
    physicsBodyDefinitions [{
      boundBone {boneName=dyn_scarf_02}
      boxExtents=<40 40 40>
      constraintSetup {angularLimitsMax=<0 5 0>, angularLimitsMin=<0 -5 -5>, angularTarget=<0 0 0>}
      localJointOffset=<6 0 0>
     }]
    relativeSpaceBone {boneName=pelvis}
   }
   n19 = Anim Dynamics - Bone: dyn_scarf_03!(ComponentPose=@n18, Alpha=0.5, GravityOverride:Vector="0,0,0", ExternalForce="0,0,0") -> (Pose) {
    angularDampingOverride=1
    angularSpringConstant=3000
    bAngularSpring=true
    bOverrideAngularDamping=true
    bUseGravityOverride=true
    simulationSpace=RootRelative
    alphaBoolBlend {blendOption=Linear}
    boundBone {boneName=dyn_scarf_03}
    componentAppliedLinearAccClamp=<1000 1000 1000>
    physicsBodyDefinitions [{
      boundBone {boneName=dyn_scarf_03}
      boxExtents=<40 40 40>
      constraintSetup {angularLimitsMax=<0 5 0>, angularLimitsMin=<0 -5 -5>, angularTarget=<0 0 0>}
      localJointOffset=<6 0 0>
     }]
    relativeSpaceBone {boneName=pelvis}
   }
   Save cached pose 'Body Dynamics'(Pose=Component To Local(ComponentPose=@n19)) {cacheName=Body Dynamics}
  ]}
 EventGraph {nodes [
   Event BlueprintUpdateAnimation()
   > n1 = Sequence!(execute) -> (then_0, then_1, then_2:Exec, then_3:Exec, then_4:Exec, then_5:Exec, then_6:Exec)
   > n2 = "Gravity Override State Function (Parameter Asset) (Nativized)"(
    CharacterPartAnimInstance=Self-Reference()
    GameTimeInSeconds=GetGameTimeInSeconds()
    Parameters=&/M_MED_OrinBlitz_Parameters
   )
   > Set GravityOverride(GravityOverride=@n2)
   > Set RBAN Settings Out(RBAN Settings Out=@n2.OutRigidBodyAnimNodeParameters)
   n3 = CheckForBackbling()
   n4 = Set Is Backbling A Cape(execute=@n1.then_1, Is Backbling A Cape:bool=OR Boolean(A=@n3, B=@n3.bIsBackblingACape)) -> (Output_Get:bool)
   > n5 = Branch(Condition=@n4)
   > Set Wear A Cape Alpha(Wear A Cape Alpha:double=0) -> (Output_Get:double)
   Set Wear A Cape Alpha(execute=@n5.else, Wear A Cape Alpha=1)
   n6 = CheckForBackbling()
   n7 = Set IsWearingBackbling(execute=@n1.then_2, IsWearingBackbling:bool=AND Boolean(A=@n6, B=NOT Boolean(A=@n6.bIsBackblingACape))) -> (Output_Get:bool)
   > Set IsWearingCape(IsWearingCape:bool=@n6.bIsBackblingACape) -> (Output_Get:bool)
   n8 = "Not Equal (Boolean)"(A=@n6.bIsBackblingACape, B=true) -> (ReturnValue=~)
   > Set ScarfEndAlpha(ScarfEndAlpha:double=SelectFloat(A=1, B=0, bPickA=AND Boolean(A=@n8, B=@n7))) -> (Output_Get:double)
   n9 = "Get Crouch "() -> ("Crouch ":bool)
   n10 = Get IsWearingBackbling() -> (IsWearingBackbling:bool)
   n11 = Get Pistol Equipped() -> (Pistol Equipped:bool)
   Set Targeting Backpack(
    execute=@n1.then_3
    Targeting Backpack:bool=AND Boolean!(A=NOT Boolean(A=@n10), B=Get bIsTargeting(), C:bool=NOT Boolean(A=@n11), D:bool=NOT Boolean(A=@n9)) -> (ReturnValue)
   ) -> (Output_Get:bool)
   n12 = Get bIsMoving2D()
   n13 = Get bIsCrouching()
   Set Crouch Walk(execute=@n1.then_5, Crouch Walk:bool=OR Boolean(A=Get bIsCrouching(), B=AND Boolean!(A=@n13, B=@n12) -> (ReturnValue))) -> (Output_Get:bool)
   Set 2d moding(execute=@n1.then_6, 2d moding:bool=AND Boolean(A=Get bIsMoving2D(), B=Get bIsInFrontEnd())) -> (Output_Get:bool)
   n14 = Switch on EFortWeaponCoreAnimation!(execute=[@n15.then, @n1.then_4], Selection:EFortWeaponCoreAnimation Enum=Get WeaponCoreAnim!() -> (WeaponCoreAnim)) -> (
    Melee:Exec
    Pistol:Exec
    Shotgun:Exec
    PaperBlueprint:Exec
    Rifle:Exec
    MeleeOneHand:Exec
    MachinePistol:Exec
    RocketLauncher:Exec
    GrenadeLauncher:Exec
    GoingCommando:Exec
    AssaultRifle:Exec
    TacticalShotgun:Exec
    SniperRifle:Exec
    TrapPlacement:Exec
    ShoulderLauncher:Exec
    AbilityDecoTool:Exec
    Crossbow:Exec
    C4:Exec
    RemoteControl:Exec
    DualWield:Exec
    AR_BullPup:Exec
    AR_ForwardGrip:Exec
    MedPackPaddles:Exec
    SMG_P90:Exec
    AR_DrumGun:Exec
    Consumable_Small:Exec
    Consumable_Large:Exec
    Balloon:Exec
    MountedTurret:Exec
    CreativeTool:Exec
    ExplosiveBow:Exec
    AshtonIndigo:Exec
    AshtonChicago:Exec
    MeleeDualWield:Exec
    Unarmed:Exec
   )
   Set isUsingRifile(execute=[@n14.AssaultRifle, @n14.Rifle, @n14.Shotgun], isUsingRifile:double=1) -> (Output_Get:double)
   Event OnWeaponChanged()
   > n15 = Set isUsingRifile(isUsingRifile=0)
  ]}
]
variables [
 2d moding {type=boolean}
 "Crouch " {type=boolean}
 Crouch Walk {type=boolean}
 GravityOverride {type=Vector}
 HeadUp Targeting {type=boolean}
 Is Backbling A Cape {type=boolean}
 isUsingRifile {type=number}
 IsWearingBackbling {type=boolean}
 IsWearingCape {type=boolean}
 Pistol Equipped {type=boolean}
 RBAN Settings Out {type=FortRigidBodyAnimNodeParameters}
 ScarfEndAlpha {type=number}
 Skydiving Backpack {type=boolean}
 Skydiving Cap {type=boolean}
 Targeting Backpack {type=boolean}
 Targeting Backpack head down {type=boolean}
 Wear A Cape Alpha {type=number}
]
